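# LWE_City_AlleySteam is a Niagara system for a city environment prop effect, named for alley steam.
# Its system scripts use SystemState and a spawn-time variable assignment, with Multiply_Float and FNTimeOfDayScalar dynamic inputs available to scale values by time of day.
# The only listed emitter, Empty, has no renderers, and no user variables or data interfaces appear, so the visible steam setup is not shown here.
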NiagaraSystem /Game/Effects/Fort_Effects/Effects/Environment/City/Props/LWE_City_AlleySteam
dependencies {
 usedDataInterfaces []
 usedDynamicInputs [&Multiply_Float, &FNTimeOfDayScalar]
 usedModules [
  &SystemState
  &self:SystemSpawnScript.SystemScriptSource.SystemScriptGraph.NiagaraNodeAssignment_0.SetVariables_#1
 ]
}
emitters {Empty {rendererClasses [], renderers []}}
summary {systemName=LWE_City_AlleySteam, emitters [{emitterName=Empty, rendererClasses []}], userVariables []}
