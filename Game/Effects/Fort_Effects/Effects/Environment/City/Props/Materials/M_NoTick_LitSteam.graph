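# Translucent, volumetric-lit steam material that fakes looping particles on instanced meshes or Niagara sprites using shader time.
# Vertex colors and object-position randomness vary looping age, velocity, size, culling and camera offset, driving sprite billboarding through world position offset.
# 4x4 SubUV smoke textures supply opacity and normals, blended with spherical normals and flattened.
# Opacity adds depth, near-camera and time-of-day controls, while quality switches simplify normals and fades on lower settings.
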Material /Game/Effects/Fort_Effects/Effects/Environment/City/Props/Materials/M_NoTick_LitSteam
graph {
 properties {
  bAutomaticallySetUsageInEditor=false
  bGenerateSphericalParticleNormals=true
  bUsedWithInstancedStaticMeshes=true
  bUsedWithNiagaraSprites=true
  blendMode=BLEND_Translucent
  translucencyLightingMode=TLM_VolumetricPerVertexDirectional
  usedShadingModels=Default Lit
 }
 nodes [
  Add(A=Time(), B=VertexColor().G)
  n1 = LinearInterpolate(
   A=ScalarParameter() {defaultValue=50, parameterName=Scale}
   B=ScalarParameter() {defaultValue=90, parameterName=Scale Max}
   Alpha=VertexColor().G
  )
  n2 = LinearInterpolate(
   A=ScalarParameter() {defaultValue=1, parameterName=Scale over life Min}
   B=ScalarParameter() {defaultValue=4, parameterName=Scale Over Life Max}
   Alpha=NamedRerouteUsage()
  )
  n3 = MaterialFunctionCall!(
   Center Location=WorldPosition().XYZ
   0-1 UVs=TextureCoordinate()
   XY scale=Multiply(A=Multiply(A=@n1, B=@n2), B=NamedRerouteUsage())
   OverridePivot
   "Normals (Optional)"
  ) {materialFunction=&Sprite}
  n4 = VertexColor()
  n5 = Multiply(A=Add(A=Time(), B=NamedRerouteUsage()), B=ScalarParameter() {defaultValue=0.1, parameterName=Delta Time Scale})
  n6 = Multiply(A=Fmod(A=Add(A=@n4.A, B=@n5), B=ScalarParameter() {defaultValue=0.5, parameterName=Age}), B=NamedRerouteUsage())
  n7 = LinearInterpolate(
   A=VectorParameter() {parameterName=Velocity Min, defaultValue=<rgba 50 0 300 0>}.RGB
   B=VectorParameter() {parameterName=VelocityMax, defaultValue=<rgba 0 0 400 0>}.RGB
   Alpha=VertexColor().G
  )
  n8 = ScalarParameter() {defaultValue=75, parameterName=Camera Offset Min}
  n9 = ScalarParameter() {defaultValue=150, parameterName=Camera Offset}
  n10 = VertexColor()
  Add(
   A=Add(A=@n3.World position offset, B=Multiply(A=@n6, B=@n7))
   B=MaterialFunctionCall(Offset Amount=LinearInterpolate(A=@n8, B=@n9, Alpha=@n10.R), World Position=WorldPosition().XYZ) {materialFunction=/Engine/Functions/Engine_MaterialFunctions02/WorldPositionOffset/CameraOffset}.Camera Offset
  )
  n11 = VectorParameter() {defaultValue=<rgba 4 4 0 0>}
  ComponentMask(None=@n11.RGB) {g=true, r=true}
  Constant()
  Constant()
  LinearInterpolate()
  MaterialFunctionCall(
   BaseNormal=TextureSample(UVs=NamedRerouteUsage()) {samplerType=SAMPLERTYPE_Normal, texture=/Game/Effects/Niagara/MasterMaterials/Smoke/Textures/NEW/RT4x4_3_Tex}.RGB
   AdditionalNormal=TextureSample() {samplerType=SAMPLERTYPE_Normal, texture=/Game/Gadgets/Assets/Teleporter/Textures/T_Spherical_Normal_02}.RGB
  ) {materialFunction=&BlendAngleCorrectedNormals}
  Multiply(B=LinearInterpolate(A=ScalarParameter() {defaultValue=0.5, parameterName=Param_2}))
  Multiply(A=Multiply(A=@n11.R, B=@n11.R), B=NamedRerouteUsage())
  Multiply(B=SphereMask(
    A=NamedRerouteUsage()
    B=ScalarParameter() {defaultValue=0.25, parameterName=Opacity Fade Start}
    Radius=ScalarParameter() {defaultValue=0.25, parameterName=Radius}
   ) {attenuationRadius=0.1, hardnessPercent=0})
  n12 = NamedRerouteUsage()
  n13 = LinearInterpolate(A=OneMinus(None=NamedRerouteUsage()), Alpha=TextureSample() {texture=&2ndInfectedMapLinear}.G)
  n14 = Constant() {r=1}
  n15 = Multiply(
   A=Multiply(
    A=TextureSample(UVs=@n12) {texture=/Game/Effects/Niagara/MasterMaterials/Smoke/Textures/NEW/RT4x4_4_Tex}.G
    B=QualitySwitch(Default=@n13, Low=@n14, High=@n13, Medium=@n14, Epic=@n13)
   )
   B=NamedRerouteUsage()
  )
  n16 = DepthFade(Opacity=@n15, FadeDistance=ScalarParameter() {defaultValue=50, parameterName=Depth Fade})
  Multiply(
   A=Multiply(A=ParticleColor().A, B=QualitySwitch(Default=@n16, Low=@n15, High=@n16, Medium=@n15, Epic=@n16))
   B=MaterialFunctionCall(Distance=ScalarParameter() {defaultValue=250, parameterName=Near Camera Fade}) {materialFunction=&Near_Camera_Fade}.Result
  )
  Multiply()
  NamedRerouteDeclaration(None=@n6) {name=Age, nodeColor=<rgba 0 0.5078 1 1>}
  n17 = ScalarParameter() {defaultValue=4}
  NamedRerouteDeclaration(None=MaterialFunctionCall!(U_Images=@n17, V_Images=@n17, SubUV Texture RGB, Time Speed=ScalarParameter() {parameterName=Param_1}, Manual Time Control, Time, UVs) {materialFunction=/Game/Effects/Fort_Effects/Materials/_Functions/MF_SubUV}.UVs) {name=Sub UV, nodeColor=<rgba 0.1953 1 0 1>}
  NamedRerouteDeclaration(None=@n3.Normals) {name=RecalculatedNormals, nodeColor=<rgba 0.9219 1 0 1>}
  n18 = NamedRerouteUsage()
  n19 = TextureSample(UVs=NamedRerouteUsage()) {samplerType=SAMPLERTYPE_Normal, texture=&RT4x4_3_Tex}
  n20 = TextureSample() {samplerType=SAMPLERTYPE_Normal, texture=&T_Spherical_Normal_02}
  n21 = Constant3Vector() {constant=<rgba 0 0 1 0>}
  n22 = MaterialFunctionCall(
   Normal=MaterialFunctionCall(
    BaseNormal=MaterialFunctionCall(BaseNormal=@n18, AdditionalNormal=@n19.RGB) {materialFunction=&BlendAngleCorrectedNormals}.Result
    AdditionalNormal=QualitySwitch(Default=@n20.RGB, Low=@n21, High=@n21, Medium=@n21, Epic=@n20.RGB)
   ) {materialFunction=&BlendAngleCorrectedNormals}.Result
   Flatness=ScalarParameter() {defaultValue=0.8, parameterName=Flatten Normal}
  ) {materialFunction=&FlattenNormal}
  n23 = Constant3Vector() {constant=<rgba 0 0 1 0>}
  NamedRerouteDeclaration(None=QualitySwitch(Default=@n22.Result, Low=@n23, High=@n23, Medium=@n23, Epic=@n22.Result)) {name=Normal, nodeColor=<rgba 0 1 0.8359 1>}
  NamedRerouteDeclaration(None=MaterialFunctionCall(
    Value_Morning=ScalarParameter() {defaultValue=1, parameterName=Opacity Morning}
    Value_Day=ScalarParameter() {defaultValue=1, parameterName=Opacity Day}
    Value_Evening=ScalarParameter() {defaultValue=1, parameterName=Opacity Evening}
    Value_Night=ScalarParameter() {defaultValue=1, parameterName=Opacity Night}
   ) {materialFunction=&Scalar_Time_of_Day}.Result) {name=TOD Opacity, nodeColor=<rgba 0.5469 0 1 1>}
  NamedRerouteDeclaration(None=Step(Y=VertexColor().A, X=ScalarParameter() {defaultValue=0.9, parameterName=Cull})) {name=Spawn Rate, nodeColor=<rgba 1 0.5391 0 1>}
  n24 = ObjectPositionWS()
  n25 = ScalarParameter() {defaultValue=500, parameterName=Randomness}
  n26 = Frac(None=Divide(A=ComponentMask(None=@n24) {r=true}, B=@n25))
  n27 = Frac(None=Divide(A=ComponentMask(None=@n24) {g=true}, B=@n25))
  n28 = Divide(A=ComponentMask(None=@n24) {b=true}, B=@n25)
  NamedRerouteDeclaration(None=Add(A=Add(A=@n26, B=@n27), B=Frac(None=@n28))) {name=Randomness, nodeColor=<rgba 0 1 0.1328 1>}
  NamedRerouteUsage()
  NamedRerouteUsage()
  TextureCoordinate()
  VectorParameter() {parameterName=Base Color, defaultValue=<rgba 1 1 1 0>}
  VertexColor()
 ]
}
